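# Masked, Default Lit material for the webbed outdoor lawn chair prop.
# A mask texture's red channel blends the diffuse texture toward a copy multiplied by the Tint vector parameter, so only masked areas are recolored.
# The red channel of a second mask texture is offset by a constant and saturated, and a normal map is sampled alongside.
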Material /SparksMaps/Environment/Props/Outdoor_LawnChairWebbed/M_Outdoor_LawnChairWebbed
graph {
 properties {bAutomaticallySetUsageInEditor=false, blendMode=BLEND_Masked, usedShadingModels=Default Lit}
 nodes [
  n1 = TextureSample() {texture=/Game/Environments/Asteria/Props/Outdoor/Outdoor_LawnChairWebbed_A/Textures/T_Outdoor_LawnChairWebbed_A_A_D}
  n2 = VectorParameter() {parameterName=Tint, defaultValue=<rgba 0.9808 0.225 1 1>}
  LinearInterpolate(
   A=@n1.RGB
   B=Multiply(A=ComponentMask(None=@n2.RGB) {b=true, g=true, r=true}, B=@n1.RGB)
   Alpha=TextureSample() {texture=/Game/Environments/Asteria/Props/Outdoor/Outdoor_LawnChairWebbed_A/Textures/T_Outdoor_LawnChairWebbed_A_A_M}.R
  )
  n3 = TextureSample() {
   samplerType=SAMPLERTYPE_Masks
   texture=/Game/Environments/Asteria/Props/Outdoor/Outdoor_LawnChairWebbed_A/Textures/T_Outdoor_LawnChairWebbed_A_A_S
  }
  Saturate(None=Add(A=@n3.R) {constB=0.3})
  TextureSample() {
   samplerType=SAMPLERTYPE_Normal
   texture=/Game/Environments/Asteria/Props/Outdoor/Outdoor_LawnChairWebbed_A/Textures/T_Outdoor_LawnChairWebbed_A_A_N
  }
 ]
}
